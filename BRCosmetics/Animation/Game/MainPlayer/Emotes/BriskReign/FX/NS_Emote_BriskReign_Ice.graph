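# Niagara system for the BriskReign emote's ice effect, surrounding the owner with ground spikes, decals, frost and shards.
# Ice_Spikes treadmills particle grids around the owner, snapping them to ground via collision traces.
# Owner-oriented region masks drive timelines that grow and retract spikes at velocity-dependent rates, and IceFrost spawns sprites from those particles.
# Front-end preview switches and time-of-day colors adjust placement, sizing and tint.
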NiagaraSystem /BRCosmetics/Animation/Game/MainPlayer/Emotes/BriskReign/FX/NS_Emote_BriskReign_Ice
dependencies {
 usedDataInterfaces [
  &NiagaraDataInterfaceCollisionQuery
  &NiagaraDataInterfaceDebugDraw
  &NiagaraDataInterfaceMeshRendererInfo
  &NiagaraDataInterfaceArrayInt32
  /Script/Niagara.NiagaraDataInterfaceArrayBool
  &NiagaraDataInterfaceArrayFloat
  &NiagaraDataInterfaceCurve
  &NiagaraDataInterfaceParticleRead
  &NiagaraDataInterfaceSocketReader
  &NiagaraDataInterfaceColorCurve
 ]
 usedDynamicInputs [&ColorFromCurve]
 usedModules [
  &SystemState
  &BlendSpaceHelper
  &self:SystemSpawnScript.SystemScriptSource.SystemScriptGraph.NiagaraNodeAssignment_0.SetVariables_#1
  &self:SystemSpawnScript.SystemScriptSource.SystemScriptGraph.NiagaraNodeAssignment_1.SetVariables_#2
 ]
}
emitters {
 Decal {
  emitterSpawnScript [/SetVariables(Emitter.WetColor:LinearColor=&MakeCustomLinearColorFromBool(
     Bool=$User.bIsFrontEndPreview
     True Color=<rgba 0.1381 0.4527 0.8814 1>
     False Color=&FNTimeOfDayColor(
      Night Color=<rgba 0.178 0.4746 0.9508 1>
      Morning Color=<rgba 0.1549 0.5266 0.9508 1>
      Day Color=<rgba 0.2206 0.5616 0.9508 1>
      Evening Color=<rgba 0.1549 0.5266 0.9508 1>
     )
    )) {isSetParameters=true}]
  emitterUpdateScript [
   &EmitterState(Life Cycle Mode=::NewEnumerator1, Loop Behavior=::NewEnumerator1, Loop Duration Mode=::NewEnumerator1, UseLoopDelay=false)
   &SpawnRate(
    Spawn Probability=&MakeCustomFloatFromBool(
     Bool=&Boolean_OrOperation(
      Bool1=&SetBoolByFloatComparison(A=&VectorLength(Vector=$Engine.Owner.Velocity), B=250)
      Bool2=&SetBoolByFloatComparison(A=&VectorLength(Vector=$Engine.Owner.Velocity), B=100, Comparison Type=::NewEnumerator6)
     )
     True Float=&FN_FrontEndFloatModifier(Gameplay=0, Front End=1)
     False Float=1
    )
    SpawnRate=5
    Use Spawn Probability=true
   )
  ]
  particleSpawnScript [
   &InitializeParticle(Color Mode=::NewEnumerator0, Lifetime Min=1.5, Lifetime Mode=::NewEnumerator1, UsePositionOffset=false, Write Material Random=true)
   /SetVariables(
    Particles.DecalOrientation:Quat4f=/Game/Effects/Niagara/Modules/MergeEmitters/MergeHelper_QuatSwitch(
     0:Quat4f=<0 1 0 1>
     1:Quat4f=<0.2 1 0 1>
     2:Quat4f=<0.4 1 0 1>
     3:Quat4f=<0.6 1 0 1>
     4:Quat4f=<0.8 1 0 1>
     5:Quat4f=?
     6:Quat4f=?
     7:Quat4f=?
     8:Quat4f=?
     9:Quat4f=?
     Index:NiagaraInt32=&SelectIntFromArray(Direct Array Index=0, Int Selection Array=DataInterface() {propertyValues {IntData [0,1,2,3,4]}})
     Num:MergeHelper_SwitcherEnum=::NewEnumerator3
    )
    Particles.DecalSize:Vector3f=&Multiply_Vector(
     A=&RandomRangeVector(Minimum=<50 150 150>, Maximum=<50 200 200>)
     B=&VectorFromFloat(Value=1.1)
    )
   ) {isSetParameters=true}
   &ShapeLocation(MaskSpawnGroup=false, Non Uniform Scale=<1 1 0.5>, Override Seed=false, Sphere Radius=40.2396, UseOwnerScale=true) {module=ShapeLocation001}
  ]
  particleUpdateScript [
   &ParticleState()
   /SetVariables(Particles.DecalColor:LinearColor=&MakeLinearColorFromVectorAndFloat(
     "Vector (RGB)"=&MakeVector(
      X=&FloatFromCurve(
       FloatCurve=DataInterface() {propertyValues {Curve {keys [<at 0.348 -0.55>,<at 0.5013 -0.55>,<at 1 0>]}}}
       Scale Curve=&RandomRangeFloat(Minimum=1.2, Maximum=0.9)
      )
      Y=0
      Z=0
     )
     "Float (Alpha)"=&Multiply_Float(
      A=&FloatFromCurve(
       FloatCurve=DataInterface() {propertyValues {Curve {
          defaultValue=1
          keys [
           {arriveTangent=5.9515, interpMode=RCIM_Cubic, leaveTangent=5.9515, tangentMode=RCTM_User, time=0, value=0}
           {interpMode=RCIM_Cubic, tangentMode=RCTM_User, time=0.5, value=1}
           {arriveTangentWeight=0.2201, interpMode=RCIM_Cubic, leaveTangentWeight=0.2332, tangentMode=RCTM_User, time=0.876, value=1}
           {arriveTangent=0.086, interpMode=RCIM_Cubic, leaveTangent=0.086, tangentMode=RCTM_User, time=1, value=0}
          ]
         }}}
       Scale Curve=&RandomRangeFloat(Minimum=0.4, Maximum=0.6)
      )
      B=&FN_FrontEndFloatModifier(Gameplay=1, Front End=5)
     )
    )) {isSetParameters=true}
   /SetVariables(Particles.Position=&FN_FrontEndVector3Modifier(
     Gameplay=Linked() {linkedVariable {name=Particles.Position, type=NiagaraPosition}}
     Front End=&AddVectorToPosition(
      Position=$Particles.Position
      Vector=&Add_Vector(
       A=&TransformVector(Vector=<0 -2 0>, Destination Space=::World)
       B=&VectorFromFloat(Value=$Engine.DeltaTime)
      )
     )
    )) {isSetParameters=true}
  ]
  rendererClasses [&NiagaraDecalRendererProperties]
  renderers [{rendererClass=&NiagaraDecalRendererProperties, rendererIndex=0}]
 }
 Frost {
  emitterUpdateScript [
   &EmitterState(Life Cycle Mode=::NewEnumerator1, Loop Behavior=::NewEnumerator1, Loop Duration Mode=::NewEnumerator1, UseLoopDelay=false)
   &SpawnBurst_Instantaneous(Spawn Count=20, Spawn Group=1, Use Spawn Probability=false, UseLoopCountLimit=false)
   &SpawnRate(SpawnRate=20, Use Spawn Probability=false)
  ]
  particleSpawnScript [
   &InitializeParticle(
    Color=&FNTimeOfDayColor(
     Night Color=<rgba 0.1466 0.4383 1 0.1>
     Morning Color=<rgba 0.3235 0.6134 1 0.1>
     Day Color=<rgba 0.3235 0.6134 1 0.1>
     Evening Color=<rgba 0.3235 0.6134 1 0.1>
    )
    Lifetime=&MaskFloat(
     Mask Source=0
     Mask Parameter=?
     Mask=0
     Masked Float=&RandomRangeFloat(Minimum=2, Maximum=1)
     Passthrough Float=&RandomRangeFloat(Minimum=0.5, Maximum=1)
    )
    Sprite Rotation Mode=::NewEnumerator2
    Sprite Size Mode=::NewEnumerator1
    Uniform Sprite Size Max=&FN_FrontEndFloatModifier(Gameplay=200, Front End=140)
    Uniform Sprite Size Min=&MaskFloat(Mask Source=0, Mask Parameter=?, Mask=0, Masked Float=100, Passthrough Float=120)
    UsePositionOffset=false
    Write Material Random=true
   )
   &ShapeLocation(
    Cone Angle=37.7614
    Cone Length=200
    MaskSpawnGroup=true
    Non Uniform Scale=<0.1 1 1>
    Offset=<0 -75.1015 15>
    Offset Coordinate Space=::Local
    Offset Mode=::NewEnumerator0
    Override Seed=false
    Rotation Mode=::NewEnumerator2
    Shape Primitive=::NewEnumerator5
    UseOwnerScale=true
    Yaw / Pitch / Roll=<0.75 0.25 0>
   )
   &ShapeLocation(
    Large Radius=30
    MaskSpawnGroup=true
    Non Uniform Scale=<1 1.3085 0.1798>
    Offset=<0 0 5>
    Offset Coordinate Space=::Local
    Offset Mode=::NewEnumerator0
    Override Seed=false
    Shape Primitive=::NewEnumerator3
    UseOwnerScale=true
   ) {module=ShapeLocation001}
   /SetVariables(Particles.Position=&MaskVector(
     Mask Source=0
     Mask Parameter=?
     Mask=0
     Masked Vector=Linked() {linkedVariable {name=Particles.ShapeLocation.ShapePosition, type=NiagaraPosition}}
     Passthrough Vector=Linked() {linkedVariable {name=Particles.ShapeLocation001.ShapePosition, type=NiagaraPosition}}
    )) {isSetParameters=true}
   &AddVelocity(
    Module.Use Falloff Distance=false
    Origin Offset=<0 0 -5>
    Override Seed=false
    Velocity Mode=::NewEnumerator1
    Velocity Speed=&MaskFloat(
     Mask Source=0
     Mask Parameter=?
     Mask=0
     Masked Float=0
     Passthrough Float=&RandomRangeFloat(Minimum=150, Maximum=250)
    )
   )
   &CameraOffset(Camera Offset Amount=&RandomRangeFloat(Minimum=0, Maximum=10))
  ]
  particleUpdateScript [
   &ParticleState()
   &ScaleSpriteSize(
    Scale Sprite Size Mode=::NewEnumerator0
    Uniform Scale Factor=&MaskFloat(
     Mask Source=0
     Mask Parameter=?
     Mask=0
     Masked Float=&FloatFromCurve(FloatCurve=DataInterface() {propertyValues {Curve {keys [
          {arriveTangent=1.2517, interpMode=RCIM_Cubic, leaveTangent=1.2517, tangentMode=RCTM_User, time=0, value=0.652}
          {arriveTangentWeight=0.534, interpMode=RCIM_Cubic, tangentMode=RCTM_User, time=0.62, value=1}
          <at 0.997 0.8784 RCIM_Cubic>
         ]}}})
     Passthrough Float=&FloatFromCurve(FloatCurve=DataInterface() {propertyValues {Curve {
         defaultValue=0.652
         keys [
          {arriveTangent=1.2755, interpMode=RCIM_Cubic, leaveTangent=1.2755, tangentMode=RCTM_User, time=0, value=0.4973}
          {interpMode=RCIM_Cubic, tangentMode=RCTM_User, time=1, value=1}
         ]
        }}})
    )
   )
   &SpriteRotationRate(Rotation Rate=&RandomRangeFloat(Minimum=-35, Maximum=35))
   &BlendSpace("Position Blend (World <-> Local)":NiagaraFloat=&MaskFloat(
     Mask Source=0
     Mask Parameter=?
     Mask=0
     Masked Float=&FloatFromCurve(FloatCurve=DataInterface() {propertyValues {Curve {keys [<at 0 1>,<at 0.2874 0>]}}})
     Passthrough Float=0
     Translation Contribution:NiagaraFloat=1
     Rotation Contribution:NiagaraFloat=1
    ))
   &ScaleColor(
    Scale Alpha=&MaskFloat(
     Mask Source=0
     Mask Parameter=?
     Mask=0
     Masked Float=&FloatFromCurve(
      FloatCurve=DataInterface() {propertyValues {Curve {
         defaultValue=1
         keys [
          {interpMode=RCIM_Cubic, tangentMode=RCTM_User, time=0, value=0}
          {interpMode=RCIM_Cubic, tangentMode=RCTM_User, time=0.5, value=1}
          {tangentMode=RCTM_User, time=1, value=0}
         ]
        }}}
      Scale Curve=0.4311
     )
     Passthrough Float=&FloatFromCurve(
      FloatCurve=DataInterface() {propertyValues {Curve {
         defaultValue=1
         keys [
          {interpMode=RCIM_Cubic, tangentMode=RCTM_User, time=0, value=1}
          {arriveTangent=-1.4546, interpMode=RCIM_Cubic, leaveTangent=-1.4546, tangentMode=RCTM_User, time=1, value=0}
         ]
        }}}
      Scale Curve=1.4
     )
    )
    ScaleA=true
    ScaleRGB=false
   )
   &Drag(
    Drag=&MaskFloat(
     Mask Source=0
     Mask Parameter=?
     Mask=0
     Masked Float=0
     Passthrough Float=&FloatFromCurve(
      FloatCurve=DataInterface() {propertyValues {Curve {
         defaultValue=1
         keys [{interpMode=RCIM_Cubic, tangentMode=RCTM_User, time=0.0681, value=0}, {interpMode=RCIM_Cubic, tangentMode=RCTM_User, time=0.203, value=1}]
        }}}
      Scale Curve=&RandomRangeFloat(Minimum=3, Maximum=7)
     )
    )
    Use Linear Drag=true
    Use Rotational Drag=true
   )
   &SolveForcesAndVelocity(Clamp Velocity=false, Limit Acceleration=false)
  ]
  rendererClasses [&NiagaraSpriteRendererProperties]
  renderers [{rendererClass=&NiagaraSpriteRendererProperties}]
 }
 IceFrost {
  emitterUpdateScript [
   &EmitterState(Life Cycle Mode=::NewEnumerator1, Loop Behavior=::NewEnumerator1, Loop Duration Mode=::NewEnumerator1, UseLoopDelay=false)
   &SpawnParticlesFromOtherEmitter(
    Attribute Reader=DataInterface() {propertyValues {EmitterBinding {bindingMode=Other, emitterName=Ice_Spikes}}}
    Spawn Rate=2
    Use Maximum Age=false
    Use Minimum Age=false
   )
  ]
  particleSpawnScript [
   &InitializeParticle(
    Lifetime Mode=::NewEnumerator1
    Sprite Rotation Mode=::NewEnumerator2
    Sprite Size Mode=::NewEnumerator1
    Uniform Sprite Size Max=50
    Uniform Sprite Size Min=70
    UsePositionOffset=false
    Write Material Random=true
   )
   &SampleParticlesFromOtherEmitter(Applied Position Offset:Vector3f=<0 0 -10>, Color Sampling=::NewEnumerator0, Transform Data Option=1)
   &CameraOffset(Camera Offset Amount=&RandomRangeFloat(Minimum=20, Maximum=10))
  ]
  particleUpdateScript [
   &ParticleState()
   &ScaleSpriteSize(Uniform Curve Sprite Scale=DataInterface() {propertyValues {Curve {keys [
        {arriveTangent=1.2517, interpMode=RCIM_Cubic, leaveTangent=1.2517, tangentMode=RCTM_User, time=0, value=0.652}
        {arriveTangentWeight=0.534, interpMode=RCIM_Cubic, tangentMode=RCTM_User, time=0.62, value=1}
        <at 0.997 0.8784 RCIM_Cubic>
       ]}}})
   &SpriteRotationRate(Rotation Rate=&RandomRangeFloat(Minimum=-35, Maximum=35))
   &RegionMask(
    Offset:Vector3f=&AddVectorToPosition(
     Position=&TransformVector(Vector=<0 0 0>, Destination Space=::World)
     Vector=&TransformVector(Vector=<0 -41.3149 0>, Source Space=::Simulation, Destination Space=::World)
    )
    Offset Space:ENiagaraCoordinateSpace=::Local
    Region Dimensions:Vector3f=&MaskVector(
     Mask Source=0
     Mask Parameter=?
     Mask=2
     Masked Vector=<210 217.7016 100>
     Passthrough Vector=&MaskVector(Mask Source=0, Mask Parameter=?, Mask=1, Masked Vector=<281.8008 459.7397 100>, Passthrough Vector=<227.8778 459.7397 100>)
    )
    Region Falloff=0
    Region Type:ENiagaraDebugDrawShape=::NewEnumerator1
    Rotation:Vector3f=&RotateVectorByQuaternion(
     Vector To Rotate=<0 1 0>
     Quaternion=&OrientMeshToVector(Look At Direction=$Engine.Owner.SystemYAxis, Reference Axis Vector=<1 0 0>)
    )
    Rotation Angle Type:ENiagara_AngleInput=::NewEnumerator0
    Transform Order:ENiagaraRegionTransformOrder=::NewEnumerator5
    UseOffset=true
    UseRotation=true
   )
   &RegionMask(
    Offset:Vector3f=&AddVectorToPosition(
     Position=&TransformVector(Vector=<0 0 0>, Destination Space=::World)
     Vector=&TransformVector(Vector=<0 19.8415 0>, Source Space=::Simulation, Destination Space=::World)
    )
    Offset Space:ENiagaraCoordinateSpace=::Local
    Region Falloff=0
    Region Radius:NiagaraFloat=&MaskFloat(Mask Source=0, Mask Parameter=?, Mask=2, Masked Float=100, Passthrough Float=122.0271)
    Rotation:Vector3f=&RotateVectorByQuaternion(
     Vector To Rotate=<0 1 0>
     Quaternion=&OrientMeshToVector(Look At Direction=$Engine.Owner.SystemYAxis, Reference Axis Vector=<1 0 0>)
    )
    Rotation Angle Type:ENiagara_AngleInput=::NewEnumerator0
    Transform Order:ENiagaraRegionTransformOrder=::NewEnumerator5
    UseOffset=true
    UseRotation=true
   ) {module=RegionMask001}
   &RegionMask(
    Offset:Vector3f=&AddVectorToPosition(
     Position=&TransformVector(Vector=<0 0 0>, Destination Space=::World)
     Vector=&TransformVector(Vector=<0 -7 0>, Source Space=::Simulation, Destination Space=::World)
    )
    Offset Space:ENiagaraCoordinateSpace=::Local
    Region Dimensions:Vector3f=&Multiply_VectorByFloat(Vector=<103.3804 262.7448 20>, Float=1)
    Region Falloff=0.4667
    Region Type:ENiagaraDebugDrawShape=::NewEnumerator1
    Rotation:Vector3f=&RotateVectorByQuaternion(
     Vector To Rotate=<0 1 0>
     Quaternion=&OrientMeshToVector(Look At Direction=$Engine.Owner.SystemYAxis, Reference Axis Vector=<1 0 0>)
    )
    Rotation Angle Type:ENiagara_AngleInput=::NewEnumerator0
    UseOffset=true
    UseRotation=true
   ) {module=RegionMask002}
   &Timeline(
    Additional Output:ENiagara_TimelineOutput=::NewEnumerator1
    Additional Output Mode:ENiagara_TimelineOutputMode=::NewEnumerator1
    Float Curve=DataInterface() {propertyValues {Curve {keys [
        {interpMode=RCIM_Cubic, leaveTangentWeight=0.0525, tangentMode=RCTM_User, time=0, value=0}
        {arriveTangent=0.1141, interpMode=RCIM_Cubic, leaveTangent=0.1141, tangentMode=RCTM_User, time=0.102, value=0.3191}
        {arriveTangent=1.3364, interpMode=RCIM_Cubic, leaveTangent=1.3364, time=0.4876, value=0.4088}
        {arriveTangent=1.1538, interpMode=RCIM_Cubic, leaveTangent=1.1538, time=0.7325, value=1.1615}
        {arriveTangentWeight=0.0746, interpMode=RCIM_Cubic, tangentMode=RCTM_User, time=1, value=1}
       ]}}}
    Play=&Boolean_OrOperation(Bool1=$Particles.RegionMask.IsInsideMask, Bool2=$Particles.RegionMask001.IsInsideMask)
    Play Rate=&MakeCustomFloatFromBool(
     Bool=&SetBoolByFloatComparison(A=&VectorLength(Vector=$User.OwnerLocalVelocity), B=250)
     True Float=10
     False Float=3
    )
    Rewind Rate:NiagaraFloat=&MakeCustomFloatFromBool(
     Bool=&SetBoolByFloatComparison(A=&VectorLength(Vector=$User.OwnerLocalVelocity), B=250)
     True Float=10
     False Float=3
    )
   )
   &ScaleColor(
    Scale Alpha=&FloatFromCurve(
     FloatCurve=DataInterface() {propertyValues {Curve {
        defaultValue=1
        keys [
         {arriveTangent=7.0386, interpMode=RCIM_Cubic, leaveTangent=7.0386, tangentMode=RCTM_User, time=0, value=0}
         {arriveTangent=0.8, interpMode=RCIM_Cubic, leaveTangent=0.8, time=0.0978, value=0.254}
         {interpMode=RCIM_Cubic, tangentMode=RCTM_User, time=0.5, value=0.4}
         <at 1 0 RCIM_Cubic>
        ]
       }}}
     Scale Curve=&FN_FrontEndFloatModifier(Gameplay=1, Front End=0.3)
    )
    Scale RGB=&MakeVectorFromLinearColorRGB(Color=&FNTimeOfDayColor(
      Night Color=<rgba 0.1827 0.3858 1 1>
      Morning Color=<rgba 0.2135 0.4089 1 1>
      Day Color=<rgba 0.2135 0.4089 1 1>
      Evening Color=<rgba 0.2135 0.4089 1 1>
     ))
    ScaleA=true
    ScaleRGB=true
   ) {module=ScaleColor001}
   &SolveForcesAndVelocity(Clamp Velocity=false, Limit Acceleration=false)
   &KillParticles(Kill Particles=&Boolean_NotOperation(Bool=&Boolean_OrOperation(Bool1=$Particles.RegionMask.IsInsideMask, Bool2=$Particles.RegionMask001.IsInsideMask)))
  ]
  rendererClasses [&NiagaraSpriteRendererProperties]
  renderers [{rendererClass=&NiagaraSpriteRendererProperties}]
 }
 Ice_Spikes {
  emitterSpawnScript [/SetVariables(
    Emitter.CrackColor:LinearColor=&MakeCustomLinearColorFromBool(Bool=$User.bIsFrontEndPreview, True Color=<rgba 0.1672 0.5966 0.8814 1>, False Color=<rgba 0.4311 0.8917 1 1>)
    Emitter.WetColor:LinearColor=&MakeCustomLinearColorFromBool(
     Bool=$User.bIsFrontEndPreview
     True Color=<rgba 0.1381 0.4527 0.8814 1>
     False Color=&FNTimeOfDayColor(
      Night Color=<rgba 0.178 0.4746 0.9508 1>
      Morning Color=<rgba 0.1549 0.5266 0.9508 1>
      Day Color=<rgba 0.2206 0.5616 0.9508 1>
      Evening Color=<rgba 0.1549 0.5266 0.9508 1>
     )
    )
   ) {isSetParameters=true}]
  emitterUpdateScript [
   &EmitterState(
    Inactive Response=::NewEnumerator1
    Life Cycle Mode=::NewEnumerator1
    Loop Behavior=::NewEnumerator1
    Loop Duration Mode=::NewEnumerator1
    UseLoopDelay=false
   )
   /SetVariables(
    Emitter.OriginOffset:Vector3f=&TransformVector(Vector=<0 0 0>, Destination Space=::World)
    Emitter.PaddingPerCell=0.8
    Emitter.VolumeRadius=300
   ) {isSetParameters=true}
   /Niagara/Modules/Spawn/Location/SpawnParticlesInGrid(
    X Count:NiagaraInt32=&MakeIntFromFloat(Float=&FN_FrontEndFloatModifier(Gameplay=14, Front End=14))
    Y Count:NiagaraInt32=&MakeIntFromFloat(Float=&FN_FrontEndFloatModifier(Gameplay=14, Front End=14))
    Z Count=0
   )
   &SpawnParticlesInGrid(
    Spawn Group=1
    X Count:NiagaraInt32=&MakeIntFromFloat(Float=&FN_FrontEndFloatModifier(Gameplay=12, Front End=12))
    Y Count:NiagaraInt32=&MakeIntFromFloat(Float=&FN_FrontEndFloatModifier(Gameplay=12, Front End=12))
    Z Count=0
   ) {module=SpawnParticlesInGrid001}
   &SpawnParticlesInGrid(
    Spawn Group=2
    X Count:NiagaraInt32=&MakeIntFromFloat(Float=&FN_FrontEndFloatModifier(Gameplay=12, Front End=12))
    Y Count:NiagaraInt32=&MakeIntFromFloat(Float=&FN_FrontEndFloatModifier(Gameplay=12, Front End=12))
    Z Count=0
   ) {module=SpawnParticlesInGrid002}
   &self:ScratchModule(DistanceMultiplier=1.1, Trace Origin Height=150, TraceDir:Vector3f=<0 0 -1>, TraceDist=300) {module=ScratchModule}
  ]
  particleSpawnScript [
   &InitializeParticle(
    Mesh Renderer Array Visibility Mode=::NewEnumerator3
    Mesh Renderer Info=DataInterface() {propertyValues {MeshRenderer=&self:Ice_Spikes_0.NiagaraMeshRendererProperties_1}}
    Mesh Scale Min=<0.75 0.75 1>
    Mesh Scale Mode=::NewEnumerator2
    Position Offset Coordinate Space=::Local
    Sprite Rotation Mode=::NewEnumerator2
    Sprite Size Mode=::NewEnumerator1
    Uniform Sprite Size Max=50
    Uniform Sprite Size Min=80
    UsePositionOffset=true
    Write Material Random=true
   )
   /SetVariables(
    Particles.DecalOrientation:Quat4f=&MergeHelper_QuatSwitch(
     0:Quat4f=<0 1 0 1>
     1:Quat4f=<0.2 1 0 1>
     2:Quat4f=<0.4 1 0 1>
     3:Quat4f=<0.6 1 0 1>
     4:Quat4f=<0.8 1 0 1>
     5:Quat4f=?
     6:Quat4f=?
     7:Quat4f=?
     8:Quat4f=?
     9:Quat4f=?
     Index:NiagaraInt32=&SelectIntFromArray(Direct Array Index=0, Int Selection Array=DataInterface() {propertyValues {IntData [0,1,2,3,4]}})
     Num:MergeHelper_SwitcherEnum=::NewEnumerator3
    )
    Particles.DecalSize:Vector3f=&Multiply_Vector(
     A=&RandomRangeVector(Minimum=<50 250 250>, Maximum=<50 200 200>)
     B=&VectorFromFloat(Value=&MaskFloat(Mask Source=0, Mask Parameter=?, Mask=1, Masked Float=1.1, Passthrough Float=1))
    )
   ) {isSetParameters=true}
   &SocketLocation(
    Orientation Sampling=::NewEnumerator1
    Position Sampling=::NewEnumerator1
    SamplePosOffset=false
    Socket Reader=$System.Socket Reader
    Socket Sampling Mode=::NewEnumerator3
    VelocitySpeedLimit=false
   )
   /Niagara/Modules/Spawn/Location/V2/GridLocation(
    Coordinate Space:ENiagaraCoordinateSpace=::World
    Module.Jitter Inside Grid Cells=false
    Module.Override Random Seed=false
    Normalize Offsets=false
    Offset:Vector3f=&Multiply_VectorByFloat(Vector=&RandomRangeVector(Minimum=<-48.9298 -27.5102 0>, Maximum=<34.6568 64.0549 1>), Float=0.7085)
    XYZ Dimensions:Vector3f=&MakeVector(
     X=&Divide_Float(A=$Emitter.VolumeRadius, B=$Emitter.PaddingPerCell)
     Y=&Divide_Float(A=$Emitter.VolumeRadius, B=$Emitter.PaddingPerCell)
     Z=0
    )
   )
   /SetVariables() {isSetParameters=true}
   &InitialMeshOrientation(
    Mesh Orientation Mode=::NewEnumerator2
    Rotation=&RandomRangeVector(Minimum=<-0.1422 0 0>, Maximum=<-0.0943 0 0>)
    Rotation Coordinate Space=::NewEnumerator0
    Use Rotation Vector=true
   )
   /SetVariables(Particles.VisibilityTag=&SpawnGroup()) {isSetParameters=true}
   &KillParticles(Kill Particles=/Niagara/DynamicInputs/Helpers/MaskBool(
     Mask Source:NiagaraInt32=0
     Mask Parameter:NiagaraInt32=?
     Mask:NiagaraInt32=0
     Masked Bool:NiagaraBool=false
     Passthrough Bool:NiagaraBool=&MaskBool(
      Mask Source:NiagaraInt32=0
      Mask Parameter:NiagaraInt32=?
      Mask:NiagaraInt32=1
      Masked Bool:NiagaraBool=/Niagara/DynamicInputs/Arrays/SelectBoolFromArray(
       Array Sampling Mode:ENiagara_LifetimeMode=::NewEnumerator1
       Direct Array Index:NiagaraInt32=?
       Direct Array Mode:ENiagara_WrapClamp=::NewEnumerator0
       Evaluation Type:ENiagaraRandomnessEvaluation=::NewEnumerator0
       Recalculate Random Each Loop:NiagaraBool=false
       Randomness Mode:ENiagaraRandomnessMode=::NewEnumerator0
       Random Seed:NiagaraInt32=0
       Fixed Random Seed:NiagaraBool=false
       Bool Selection Array:NiagaraDataInterfaceArrayBool=DataInterface() {propertyValues {GpuSyncMode=SyncCpuToGpu, MaxElements=0, BoolData [true,true,true,true,true,false,false,false,false,false,false,false]}}
      )
      Passthrough Bool:NiagaraBool=&SelectBoolFromArray(
       Array Sampling Mode:ENiagara_LifetimeMode=::NewEnumerator1
       Direct Array Index:NiagaraInt32=?
       Direct Array Mode:ENiagara_WrapClamp=::NewEnumerator0
       Evaluation Type:ENiagaraRandomnessEvaluation=::NewEnumerator0
       Recalculate Random Each Loop:NiagaraBool=false
       Randomness Mode:ENiagaraRandomnessMode=::NewEnumerator0
       Random Seed:NiagaraInt32=0
       Fixed Random Seed:NiagaraBool=false
       Bool Selection Array:NiagaraDataInterfaceArrayBool=DataInterface() {propertyValues {GpuSyncMode=SyncCpuToGpu, MaxElements=0, BoolData [true,true,true,true,true,true,false,true,false,false,false,false]}}
      )
     )
    ))
  ]
  particleUpdateScript [
   &ParticleState(Kill Particles When Lifetime Has Elapsed=false, Let Infinitely Lived Particles Die When Emitter Deactivates=true)
   &SolveForcesAndVelocity(Clamp Velocity=false, Limit Acceleration=false)
   /SetVariables(Particles.PositionCenter:NiagaraPosition=&AddVectorToPosition(Position=$Engine.Owner.Position, Vector=$Emitter.OriginOffset)) {isSetParameters=true}
   &self:TreadmillParticles_01(
    GridSize:NiagaraFloat=&Multiply_Float(A=$Emitter.VolumeRadius, B=2)
    InputPosition:NiagaraPosition=$Particles.PositionCenter
   ) {module=TreadmillParticles_01}
   &DoOnce(
    Reset:NiagaraBool=&SetBoolByFloatComparison(A=$Particles.Timeline001.Time, B=0.9)
    Trigger Condition:NiagaraBool=&SetBoolByFloatComparison(
     A=&VectorLength(Vector=Linked() {linkedVariable {name=Particles.Position, type=NiagaraPosition}})
     B=&VectorLength(Vector=Linked() {linkedVariable {name=Particles.Previous.Position, type=NiagaraPosition}})
     Comparison Type=::NewEnumerator5
    )
   )
   &Timeline(Instantaneous Rewind=true, Loop While Playing=true, Play=$Particles.DoOnce.TriggeredOnce, Play Rate=10) {module=Timeline001}
   /SetVariables(Particles.PositionNoCollision:NiagaraPosition=$StackContext.Position) {isSetParameters=true}
   &self:GridCollision(
    CharacterPos:NiagaraPosition=&AddVectorToPosition(Position=$Engine.Owner.Position, Vector=<0 0 50>)
    Disable Trace:NiagaraBool=&Boolean_AndOperation(
     Bool1=&Boolean_NotOperation(Bool=$Particles.DoOnce.TriggeredOnce)
     Bool2=&SetBoolByFloatComparison(A=$Emitter.Age, B=1)
    )
    Height Adjustment=1
    Trace Origin Height=150
    TraceDir:Vector3f=<0 0 -1>
    TraceDist=250
    bIsFrontEnd:NiagaraBool=$User.bIsFrontEndPreview
   ) {module=GridCollision}
   /SetVariables(Particles.Position=&FN_FrontEndVector3Modifier(
     Gameplay=/Niagara/DynamicInputs/TypeConversions/MakeCustomPositionFromBool(
      Bool:NiagaraBool=$Particles.DoOnce.TriggeredOnce
      True Position:NiagaraPosition=$Particles.CollisionPos
      False Position:NiagaraPosition=&MakeCustomPositionFromBool(
       Bool:NiagaraBool=&SetBoolByFloatComparison(A=$Emitter.Age, B=1)
       True Position:NiagaraPosition=$Particles.Previous.Position
       False Position:NiagaraPosition=$Particles.CollisionPos
      )
     )
     Front End=&AddVectorToPosition(
      Position=&MakeVector(
       X=&MakeFloatFromVector(Vector=Linked() {linkedVariable {name=Particles.PositionNoCollision, type=NiagaraPosition}}, Channel=::NewEnumerator0)
       Y=&MakeFloatFromVector(Vector=Linked() {linkedVariable {name=Particles.PositionNoCollision, type=NiagaraPosition}}, Channel=::NewEnumerator1)
       Z=&Add_Float(
        A=&MakeFloatFromVector(Vector=Linked() {linkedVariable {name=Engine.Emitter.SimulationPosition, type=NiagaraPosition}}, Channel=::NewEnumerator2)
        B=-5
       )
      )
      Vector=&RotateVectorByQuaternion(
       Vector To Rotate=&Multiply_VectorByFloat(Vector=&TransformVector(Vector=<0 -60 0>, Destination Space=::World), Float=$Engine.DeltaTime)
       Quaternion=$Output.SocketLocation.SampledRotation
      )
     )
    )) {isSetParameters=true}
   &RegionMask(
    Offset:Vector3f=&AddVectorToPosition(
     Position=Linked() {linkedVariable {name=Emitter.OriginOffset, type=Vector3f}}
     Vector=&TransformVector(Vector=<0 -41.3149 0>, Source Space=::Simulation, Destination Space=::World)
    )
    Offset Space:ENiagaraCoordinateSpace=::Local
    Region Dimensions:Vector3f=&FN_FrontEndVector3Modifier(
     Gameplay=&MaskVector(
      Mask Source=0
      Mask Parameter=?
      Mask=2
      Masked Vector=<210 390 100>
      Passthrough Vector=&MaskVector(Mask Source=0, Mask Parameter=?, Mask=1, Masked Vector=<281.8008 459.7397 100>, Passthrough Vector=<227.8778 459.7397 100>)
     )
     Front End=&MaskVector(
      Mask Source=0
      Mask Parameter=?
      Mask=2
      Masked Vector=<190 300 100>
      Passthrough Vector=&MaskVector(Mask Source=0, Mask Parameter=?, Mask=1, Masked Vector=<281.8008 459.7397 100>, Passthrough Vector=<214.7707 415.6273 100>)
     )
    )
    Region Falloff=0
    Region Type:ENiagaraDebugDrawShape=::NewEnumerator1
    Rotation:Vector3f=&RotateVectorByQuaternion(
     Vector To Rotate=<0 1 0>
     Quaternion=&OrientMeshToVector(Look At Direction=$Engine.Owner.SystemYAxis, Reference Axis Vector=<1 0 0>)
    )
    Rotation Angle Type:ENiagara_AngleInput=::NewEnumerator0
    Transform Order:ENiagaraRegionTransformOrder=::NewEnumerator5
    UseOffset=true
    UseRotation=true
   )
   &RegionMask(
    Offset:Vector3f=&AddVectorToPosition(
     Position=Linked() {linkedVariable {name=Emitter.OriginOffset, type=Vector3f}}
     Vector=&TransformVector(Vector=<0 19.8415 0>, Source Space=::Simulation, Destination Space=::World)
    )
    Offset Space:ENiagaraCoordinateSpace=::Local
    Region Falloff=0
    Region Radius:NiagaraFloat=&FN_FrontEndFloatModifier(
     Gameplay=&MaskFloat(Mask Source=0, Mask Parameter=?, Mask=2, Masked Float=100, Passthrough Float=122.0271)
     Front End=&MaskFloat(Mask Source=0, Mask Parameter=?, Mask=2, Masked Float=70, Passthrough Float=95)
    )
    Rotation:Vector3f=&RotateVectorByQuaternion(
     Vector To Rotate=<0 1 0>
     Quaternion=&OrientMeshToVector(Look At Direction=$Engine.Owner.SystemYAxis, Reference Axis Vector=<1 0 0>)
    )
    Rotation Angle Type:ENiagara_AngleInput=::NewEnumerator0
    Transform Order:ENiagaraRegionTransformOrder=::NewEnumerator5
    UseOffset=true
    UseRotation=true
   ) {module=RegionMask001}
   &RegionMask(
    Offset:Vector3f=&AddVectorToPosition(
     Position=Linked() {linkedVariable {name=Emitter.OriginOffset, type=Vector3f}}
     Vector=&TransformVector(Vector=<0 -7 0>, Source Space=::Simulation, Destination Space=::World)
    )
    Offset Space:ENiagaraCoordinateSpace=::Local
    Region Dimensions:Vector3f=&Multiply_VectorByFloat(
     Vector=&FN_FrontEndVector3Modifier(Gameplay=<103.3804 262.7448 20>, Front End=<177.1052 247.3306 40>)
     Float=1
    )
    Region Falloff=0.4667
    Region Type:ENiagaraDebugDrawShape=::NewEnumerator1
    Rotation:Vector3f=&RotateVectorByQuaternion(
     Vector To Rotate=<0 1 0>
     Quaternion=&OrientMeshToVector(Look At Direction=$Engine.Owner.SystemYAxis, Reference Axis Vector=<1 0 0>)
    )
    Rotation Angle Type:ENiagara_AngleInput=::NewEnumerator0
    UseOffset=true
    UseRotation=true
   ) {module=RegionMask002}
   &Timeline(
    Additional Output:ENiagara_TimelineOutput=::NewEnumerator1
    Additional Output Mode:ENiagara_TimelineOutputMode=::NewEnumerator1
    Float Curve=DataInterface() {propertyValues {Curve {keys [
        {interpMode=RCIM_Cubic, leaveTangentWeight=0.0525, tangentMode=RCTM_User, time=0, value=0}
        {arriveTangent=0.1141, interpMode=RCIM_Cubic, leaveTangent=0.1141, tangentMode=RCTM_User, time=0.102, value=0.3191}
        {arriveTangent=1.3364, interpMode=RCIM_Cubic, leaveTangent=1.3364, time=0.4876, value=0.4088}
        {arriveTangent=1.1538, interpMode=RCIM_Cubic, leaveTangent=1.1538, time=0.7324, value=1.1615}
        {arriveTangentWeight=0.0746, interpMode=RCIM_Cubic, tangentMode=RCTM_User, time=1, value=1}
       ]}}}
    Play=&Boolean_OrOperation(Bool1=$Particles.RegionMask.IsInsideMask, Bool2=$Particles.RegionMask001.IsInsideMask)
    Play Rate=&MakeCustomFloatFromBool(
     Bool=&SetBoolByFloatComparison(A=&VectorLength(Vector=$User.OwnerLocalVelocity), B=250)
     True Float=10
     False Float=3
    )
    Rewind Rate:NiagaraFloat=&MakeCustomFloatFromBool(
     Bool=&SetBoolByFloatComparison(A=&VectorLength(Vector=$User.OwnerLocalVelocity), B=250)
     True Float=10
     False Float=3
    )
   )
   &Timeline(
    Additional Output:ENiagara_TimelineOutput=::NewEnumerator1
    Additional Output Mode:ENiagara_TimelineOutputMode=::NewEnumerator1
    Float Curve=DataInterface() {propertyValues {Curve {defaultValue=0, keys [{tangentMode=RCTM_User, time=0, value=0}, {tangentMode=RCTM_User, time=1, value=1}]}}}
    Play=&Boolean_OrOperation(Bool1=$Particles.RegionMask.IsInsideMask, Bool2=$Particles.RegionMask001.IsInsideMask)
    Play Rate=&MaskFloatBySpawnGroup(Masked Float=1, Passthrough Float=2, Spawn Group Mask=1)
    Rewind Rate=3
   ) {module=Timeline002}
   &UpdateMeshOrientation(
    Facing Direction=&FN_FrontEndVector3Modifier(
     Gameplay=&MakeCustomVectorFromBool(
      Bool=&Boolean_NotOperation(Bool=&Boolean_OrOperation(Bool1=$Particles.RegionMask.IsInsideMask, Bool2=$Particles.RegionMask001.IsInsideMask))
      True Vector=$Engine.Owner.SystemXAxis
      False Vector=&MakeCustomVectorFromBool(
       Bool=$Output.ParticleState.FirstFrame
       True Vector=$Engine.Owner.SystemXAxis
       False Vector=&ReturnMeshOrientationAxis(Mesh Orientation Axis=::XAxis, Mesh Orientation Quat=$Particles.MeshOrientation)
      )
     )
     Front End=$Engine.Owner.SystemXAxis
    )
    Facing Interpolation=false
    Orientation Method=::NewEnumerator1
    Override Quat=false
    TransformFacing=false
    TransformUp=false
   )
   &ScaleMeshSize(Scale Factor=&VectorFromFloat(Value=&Multiply_Float(
      A=&Multiply_Float(
       A=&Multiply_Float(
        A=$Particles.Timeline.Float
        B=&FloatFromCurve(
         FloatCurve=DataInterface() {propertyValues {Curve {
            defaultValue=0
            keys [
             {arriveTangent=0.4008, interpMode=RCIM_Cubic, leaveTangent=0.4008, tangentMode=RCTM_User, time=0, value=0}
             {arriveTangent=1.4557, interpMode=RCIM_Cubic, leaveTangent=1.4557, tangentMode=RCTM_User, time=0.3125, value=0.3518}
             {interpMode=RCIM_Cubic, tangentMode=RCTM_User, time=1, value=1}
            ]
           }}}
         CurveIndex=&Add_Float(A=$Particles.NormalizedAge, B=&RandomRangeFloat(Minimum=0, Maximum=-0.5))
        )
       )
       B=&FloatFromCurve(
        FloatCurve=DataInterface() {propertyValues {Curve {
           defaultValue=1
           keys [
            {interpMode=RCIM_Cubic, leaveTangentWeight=0.0978, tangentMode=RCTM_User, time=0.0166, value=0.5624}
            {arriveTangentWeight=0.0978, interpMode=RCIM_Cubic, leaveTangentWeight=0.23, tangentMode=RCTM_User, time=0.31, value=1}
            {arriveTangent=-0.3743, arriveTangentWeight=0.2456, interpMode=RCIM_Cubic, leaveTangent=-0.3743, tangentMode=RCTM_User, time=1, value=0.49}
           ]
          }}}
        CurveIndex=&RemapRange(
         Input Value To Remap=&DistanceBetweenPositions(
          A=&AddVectorToPosition(
           Position=$Engine.Emitter.SimulationPosition
           Vector=&TransformVector(Vector=<0 100 0>, Destination Space=::World)
          )
          B=$Particles.Position
         )
         Input Min=0
         Input Max=300
         Output Min=0
         Output Max=1
        )
        Scale Curve=&RandomRangeFloat(Minimum=0.7, Maximum=1)
       )
      )
      B=&FloatFromCurve(
       FloatCurve=DataInterface() {propertyValues {Curve {defaultValue=0, keys [{tangentMode=RCTM_User, time=0, value=1}, {tangentMode=RCTM_User, time=1, value=0}]}}}
       CurveIndex=$Particles.RegionMask002.Float
      )
     ))) {module=ScaleMeshSize001}
   &ScaleColor(
    Scale Alpha=&Multiply_Float(
     A=&Multiply_Float(
      A=&Multiply_Float(
       A=$Particles.Timeline.Float
       B=&FloatFromCurve(FloatCurve=DataInterface() {propertyValues {Curve {
           defaultValue=0
           keys [
            {arriveTangent=2.9083, interpMode=RCIM_Cubic, leaveTangent=2.9083, tangentMode=RCTM_User, time=0, value=0}
            {interpMode=RCIM_Cubic, tangentMode=RCTM_User, time=1, value=1}
           ]
          }}})
      )
      B=&FloatFromCurve(
       FloatCurve=DataInterface() {propertyValues {Curve {
          defaultValue=1
          keys [
           {interpMode=RCIM_Cubic, leaveTangentWeight=0.0978, tangentMode=RCTM_User, time=0.0166, value=0.5624}
           {arriveTangentWeight=0.0978, interpMode=RCIM_Cubic, leaveTangentWeight=0.23, tangentMode=RCTM_User, time=0.31, value=1}
           {arriveTangent=-0.3743, arriveTangentWeight=0.2456, interpMode=RCIM_Cubic, leaveTangent=-0.3743, tangentMode=RCTM_User, time=1, value=0.49}
          ]
         }}}
       CurveIndex=&RemapRange(
        Input Value To Remap=&DistanceBetweenPositions(
         A=&AddVectorToPosition(
          Position=$Engine.Emitter.SimulationPosition
          Vector=&TransformVector(Vector=<0 100 0>, Destination Space=::World)
         )
         B=$Particles.Position
        )
        Input Min=0
        Input Max=300
        Output Min=0
        Output Max=1
       )
       Scale Curve=&RandomRangeFloat(Minimum=0.7, Maximum=1)
      )
     )
     B=&FloatFromCurve(
      FloatCurve=DataInterface() {propertyValues {Curve {defaultValue=0, keys [{tangentMode=RCTM_User, time=0, value=1}, {tangentMode=RCTM_User, time=1, value=0}]}}}
      CurveIndex=$Particles.RegionMask002.Float
     )
    )
    Scale RGB=&MakeVectorFromLinearColorRGB(Color=&FNTimeOfDayColor(Night Color=<rgba 0.217 0.5708 1 1>, Morning Color=<rgba 1 1 1 1>, Day Color=<rgba 1 1 1 1>, Evening Color=<rgba 1 1 1 1>))
    ScaleA=true
    ScaleRGB=true
   )
   /SetVariables(Particles.DecalColor:LinearColor=&MakeLinearColorFromVectorAndFloat(
     "Vector (RGB)"=&MakeVector(
      X=&FloatFromCurve(
       FloatCurve=DataInterface() {propertyValues {Curve {keys [<at 0.1848 0>,<at 0.218 -0.55>,<at 0.7383 -0.55>,<at 1 0>]}}}
       CurveIndex=&RemapRange(
        Input Value To Remap=&DistanceBetweenPositions(
         A=&AddVectorToPosition(
          Position=$Engine.Emitter.SimulationPosition
          Vector=&TransformVector(Vector=<0 150 0>, Destination Space=::World)
         )
         B=$Particles.Position
        )
        Input Min=0
        Input Max=&MaskFloat(Mask Source=0, Mask Parameter=?, Mask=1, Masked Float=300, Passthrough Float=100)
        Output Min=0
        Output Max=1
       )
       Scale Curve=&RandomRangeFloat(Minimum=1.2, Maximum=0.9)
      )
      Y=&RandomRangeFloat(Minimum=0, Maximum=1)
      Z=0
     )
     "Float (Alpha)"=&Multiply_Float(
      A=&FloatFromCurve(
       FloatCurve=DataInterface() {propertyValues {Curve {
          defaultValue=1
          keys [
           {arriveTangent=1.6035, interpMode=RCIM_Cubic, leaveTangent=1.6035, tangentMode=RCTM_User, time=0, value=0}
           {arriveTangentWeight=0.2201, interpMode=RCIM_Cubic, leaveTangentWeight=0.2332, tangentMode=RCTM_User, time=0.9544, value=1}
           {arriveTangent=0.086, interpMode=RCIM_Cubic, leaveTangent=0.086, tangentMode=RCTM_User, time=1, value=1}
          ]
         }}}
       CurveIndex=$Particles.Timeline002.Time
       Scale Curve=&FNTimeOfDayScalar(Night Scale=0.6, Morning Scale=0.9, Evening Scale=0.9)
      )
      B=&MaskFloat(
       Mask Source=0
       Mask Parameter=?
       Mask=1
       Masked Float=&RandomRangeFloat(Minimum=0.4, Maximum=0.6)
       Passthrough Float=1
      )
     )
    )) {isSetParameters=true}
  ]
  rendererClasses [&NiagaraMeshRendererProperties, &NiagaraDecalRendererProperties]
  renderers [
   {rendererClass=&NiagaraMeshRendererProperties, rendererIndex=0}
   {rendererClass=&NiagaraDecalRendererProperties, rendererIndex=1}
   {rendererClass=&NiagaraDecalRendererProperties, rendererIndex=2}
  ]
 }
 Light {rendererClasses [], renderers []}
 Shards {
  emitterUpdateScript [
   &EmitterState(Life Cycle Mode=::NewEnumerator1, Loop Behavior=::NewEnumerator1, Loop Duration Mode=::NewEnumerator1, UseLoopDelay=false)
   &SpawnBurst_Instantaneous(Spawn Count=25, Spawn Group=1, Spawn Time=0.1, Use Spawn Probability=false, UseLoopCountLimit=false)
   &SpawnRate(
    SpawnRate=&MakeCustomFloatFromBool(
     Bool=&SetBoolByFloatComparison(A=&VectorLength(Vector=$User.OwnerLocalVelocity), B=50)
     True Float=15
     False Float=0
    )
    Use Spawn Probability=false
   )
  ]
  particleSpawnScript [
   &InitializeParticle(
    Color Mode=::NewEnumerator0
    Lifetime Max=0.3
    Lifetime Min=0.7
    Lifetime Mode=::NewEnumerator1
    Mesh Renderer Array Visibility Mode=::NewEnumerator3
    Mesh Renderer Info=DataInterface() {propertyValues {MeshRenderer=&self:Shards_0.NiagaraMeshRendererProperties_2}}
    Mesh Scale Mode=::NewEnumerator1
    Mesh Uniform Scale Max=0.6
    Mesh Uniform Scale Min=0.4
    UsePositionOffset=false
    Write Material Random=true
   )
   &ShapeLocation(
    MaskSpawnGroup=false
    Non Uniform Scale=<1 1 0.2>
    Offset=<0 0 5>
    Offset Mode=::NewEnumerator0
    Override Seed=false
    Sphere Radius=150
    Sphere Surface Distribution=0.5939
    UseOwnerScale=true
   )
   &AddVelocity(
    Velocity=&MaskVector(
     Mask Source=0
     Mask Parameter=?
     Mask=0
     Masked Vector=&RandomRangeVector(Minimum=<0.2 -0.8 1>, Maximum=<-0.2 -0.4 1>)
     Passthrough Vector=&RandomRangeVector(Minimum=<0.2 -0.1 1>, Maximum=<-0.2 0.7 1>)
    )
    Velocity Speed Scale=&RandomRangeFloat(Minimum=300, Maximum=500)
   )
   &InitialMeshOrientation(Use Rotation Vector=false)
  ]
  particleUpdateScript [
   &ParticleState()
   &RegionMask(
    Offset:Vector3f=&AddVectorToPosition(
     Position=&TransformVector(Vector=<0 0 0>, Destination Space=::World)
     Vector=&TransformVector(Vector=<0 -41.3149 0>, Source Space=::Simulation, Destination Space=::World)
    )
    Offset Space:ENiagaraCoordinateSpace=::Local
    Region Dimensions:Vector3f=&MaskVector(
     Mask Source=0
     Mask Parameter=?
     Mask=2
     Masked Vector=<210 217.7016 100>
     Passthrough Vector=&MaskVector(Mask Source=0, Mask Parameter=?, Mask=1, Masked Vector=<281.8008 459.7397 100>, Passthrough Vector=<227.8778 459.7397 100>)
    )
    Region Falloff=0
    Region Type:ENiagaraDebugDrawShape=::NewEnumerator1
    Rotation:Vector3f=&RotateVectorByQuaternion(
     Vector To Rotate=<0 1 0>
     Quaternion=&OrientMeshToVector(Look At Direction=$Engine.Owner.SystemYAxis, Reference Axis Vector=<1 0 0>)
    )
    Rotation Angle Type:ENiagara_AngleInput=::NewEnumerator0
    Transform Order:ENiagaraRegionTransformOrder=::NewEnumerator5
    UseOffset=true
    UseRotation=true
   )
   &RegionMask(
    Offset:Vector3f=&AddVectorToPosition(
     Position=&TransformVector(Vector=<0 0 0>, Destination Space=::World)
     Vector=&TransformVector(Vector=<0 19.8415 0>, Source Space=::Simulation, Destination Space=::World)
    )
    Offset Space:ENiagaraCoordinateSpace=::Local
    Region Falloff=0
    Region Radius:NiagaraFloat=&MaskFloat(Mask Source=0, Mask Parameter=?, Mask=2, Masked Float=100, Passthrough Float=122.0271)
    Rotation:Vector3f=&RotateVectorByQuaternion(
     Vector To Rotate=<0 1 0>
     Quaternion=&OrientMeshToVector(Look At Direction=$Engine.Owner.SystemYAxis, Reference Axis Vector=<1 0 0>)
    )
    Rotation Angle Type:ENiagara_AngleInput=::NewEnumerator0
    Transform Order:ENiagaraRegionTransformOrder=::NewEnumerator5
    UseOffset=true
    UseRotation=true
   ) {module=RegionMask001}
   &RegionMask(
    Offset:Vector3f=&AddVectorToPosition(
     Position=&TransformVector(Vector=<0 0 0>, Destination Space=::World)
     Vector=&TransformVector(Vector=<0 -7 0>, Source Space=::Simulation, Destination Space=::World)
    )
    Offset Space:ENiagaraCoordinateSpace=::Local
    Region Dimensions:Vector3f=&Multiply_VectorByFloat(Vector=<103.3804 262.7448 20>, Float=1)
    Region Falloff=0.4667
    Region Type:ENiagaraDebugDrawShape=::NewEnumerator1
    Rotation:Vector3f=&RotateVectorByQuaternion(
     Vector To Rotate=<0 1 0>
     Quaternion=&OrientMeshToVector(Look At Direction=$Engine.Owner.SystemYAxis, Reference Axis Vector=<1 0 0>)
    )
    Rotation Angle Type:ENiagara_AngleInput=::NewEnumerator0
    UseOffset=true
    UseRotation=true
   ) {module=RegionMask002}
   &Timeline(
    Additional Output:ENiagara_TimelineOutput=::NewEnumerator1
    Additional Output Mode:ENiagara_TimelineOutputMode=::NewEnumerator1
    Float Curve=DataInterface() {propertyValues {Curve {keys [
        {interpMode=RCIM_Cubic, leaveTangentWeight=0.0525, tangentMode=RCTM_User, time=0, value=0}
        {arriveTangent=0.1141, interpMode=RCIM_Cubic, leaveTangent=0.1141, tangentMode=RCTM_User, time=0.102, value=0.3191}
        {arriveTangent=1.3364, interpMode=RCIM_Cubic, leaveTangent=1.3364, time=0.4876, value=0.4088}
        {arriveTangent=1.1538, interpMode=RCIM_Cubic, leaveTangent=1.1538, time=0.7325, value=1.1615}
        {arriveTangentWeight=0.0746, interpMode=RCIM_Cubic, tangentMode=RCTM_User, time=1, value=1}
       ]}}}
    Play=&Boolean_OrOperation(Bool1=$Particles.RegionMask.IsInsideMask, Bool2=$Particles.RegionMask001.IsInsideMask)
    Play Rate=&MakeCustomFloatFromBool(
     Bool=&SetBoolByFloatComparison(A=&VectorLength(Vector=$User.OwnerLocalVelocity), B=250)
     True Float=10
     False Float=3
    )
    Rewind Rate:NiagaraFloat=&MakeCustomFloatFromBool(
     Bool=&SetBoolByFloatComparison(A=&VectorLength(Vector=$User.OwnerLocalVelocity), B=250)
     True Float=10
     False Float=3
    )
   )
   &GravityForce()
   &Drag(
    Drag=&FloatFromCurve(
     FloatCurve=DataInterface() {propertyValues {Curve {
        defaultValue=1
        keys [
         {interpMode=RCIM_Cubic, tangentMode=RCTM_User, time=0.0365, value=0.2703}
         {interpMode=RCIM_Cubic, tangentMode=RCTM_User, time=0.1454, value=1}
        ]
       }}}
     Scale Curve=&RandomRangeFloat(Minimum=7, Maximum=5)
    )
    Use Linear Drag=true
    Use Rotational Drag=true
   )
   &SolveForcesAndVelocity(Clamp Velocity=false, Limit Acceleration=false)
   &ScaleMeshSize(Scale Factor=&VectorFromFloat(Value=&FloatFromCurve(FloatCurve=DataInterface() {propertyValues {Curve {
         defaultValue=1
         keys [
          {interpMode=RCIM_Cubic, tangentMode=RCTM_User, time=0, value=0.5628}
          {interpMode=RCIM_Cubic, tangentMode=RCTM_User, time=0.5, value=1}
          {tangentMode=RCTM_User, time=1, value=0}
         ]
        }}})))
   &ScaleMeshSizeBySpeed(Max Scale Factor=<1 1 2>)
  ]
  rendererClasses [&NiagaraMeshRendererProperties]
  renderers [{rendererClass=&NiagaraMeshRendererProperties, rendererIndex=0}]
 }
}
summary {
 systemName=NS_Emote_BriskReign_Ice
 emitters [
  {emitterName=Ice_Spikes, rendererClasses [&NiagaraMeshRendererProperties, &NiagaraDecalRendererProperties]}
  {emitterName=Decal, rendererClasses [&NiagaraDecalRendererProperties]}
  {emitterName=Frost, rendererClasses [&NiagaraSpriteRendererProperties]}
  {emitterName=IceFrost, rendererClasses [&NiagaraSpriteRendererProperties]}
  {emitterName=Shards, rendererClasses [&NiagaraMeshRendererProperties]}
  {emitterName=Light, rendererClasses []}
 ]
 userVariables [
  User.bIsFrontEnd {type=NiagaraBool}
  User.bIsFrontEndPreview {type=NiagaraBool}
  User.OwnerLocalVelocity {type=Vector3f, description="", defaultValue {struct=Vector3f, value=<1 1 1>}}
 ]
}
