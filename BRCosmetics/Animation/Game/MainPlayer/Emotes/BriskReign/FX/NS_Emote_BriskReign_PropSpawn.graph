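# Niagara system for the prop-spawn effect of the BriskReign emote, made of a Frost emitter and two sparkle emitters.
# A system spawn Set Variables module runs alongside System State, with dynamic inputs for time-of-day color and color/Vector4 conversion.
# The SparklesLWE names suggest lightweight emitters.
# No user parameters, data interfaces or renderers are listed.
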NiagaraSystem /BRCosmetics/Animation/Game/MainPlayer/Emotes/BriskReign/FX/NS_Emote_BriskReign_PropSpawn
dependencies {
 usedDataInterfaces []
 usedDynamicInputs [
  /Niagara/DynamicInputs/LinearColor/MakeLinearColorFromVector4
  &MakeVector4_FromLinearColor
  &FNTimeOfDayColor
 ]
 usedModules [
  &self:SystemSpawnScript.SystemScriptSource.SystemScriptGraph.NiagaraNodeAssignment_0.SetVariables_#1
  &SystemState
 ]
}
emitters {Frost {rendererClasses [], renderers []}, SparklesLWE {rendererClasses [], renderers []}, SparklesLWE001 {rendererClasses [], renderers []}}
summary {
 systemName=NS_Emote_BriskReign_PropSpawn
 emitters [{emitterName=Frost, rendererClasses []}, {emitterName=SparklesLWE, rendererClasses []}, {emitterName=SparklesLWE001, rendererClasses []}]
 userVariables []
}
